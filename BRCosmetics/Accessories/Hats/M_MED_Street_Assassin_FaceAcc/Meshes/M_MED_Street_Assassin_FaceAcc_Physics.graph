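# Physics asset for the Street Assassin face accessory cosmetic, defining collision and joint limits for its hair and ponytail bones.
# Capsule bodies cover the head, four left and right hair strand bones, and two ponytail bones.
# Every hair and ponytail body constrains directly to the head rather than chaining, including ponytail_B.
# Ponytail joints allow wider swing than the side strands, while all joints share a tight twist limit.
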PhysicsAsset /BRCosmetics/Accessories/Hats/M_MED_Street_Assassin_FaceAcc/Meshes/M_MED_Street_Assassin_FaceAcc_Physics
bodies {
 dyn_hairA_l [{shapeType=Capsule, length=2.75, radius=2, center=<0 0 0.4106>, rotation=<rot 0 -90 0>}]
 dyn_hairA_r [{shapeType=Capsule, length=2.75, radius=2, center=<0.174 0 -0.5096>, rotation=<rot 0 -90 0>}]
 dyn_hairB_l [{shapeType=Capsule, length=2.75, radius=2, center=<0.0408 -0.0378 0.3666>, rotation=<rot 0 -90 0>}]
 dyn_hairB_r [{shapeType=Capsule, length=2.75, radius=2, center=<-0.0887 0 -0.5965>, rotation=<rot 0 -90 0>}]
 head [{shapeType=Capsule, length=9.9077, radius=6.3987, rotation=<rot 90 0 0>}]
 ponytail_A [{shapeType=Capsule, length=3.75, radius=2.5, center=<1.9794 0.0001 0>, rotation=<rot 90 0 0>}]
 ponytail_B [{shapeType=Capsule, length=4.75, radius=3.25, center=<1.9794 0.0001 0>, rotation=<rot 90 0 0>}]
}
constraints [
 {bone1Name=ponytail_A, bone2Name=head, swing1LimitDegrees=80, swing2LimitDegrees=80, twistLimitDegrees=10}
 {bone1Name=ponytail_B, bone2Name=head, swing1LimitDegrees=80, swing2LimitDegrees=80, twistLimitDegrees=10}
 {bone1Name=dyn_hairA_l, bone2Name=head, swing1LimitDegrees=40, swing2LimitDegrees=40, twistLimitDegrees=10}
 {bone1Name=dyn_hairA_r, bone2Name=head, swing1LimitDegrees=40, swing2LimitDegrees=40, twistLimitDegrees=10}
 {bone1Name=dyn_hairB_l, bone2Name=head, swing1LimitDegrees=40, swing2LimitDegrees=40, twistLimitDegrees=10}
 {bone1Name=dyn_hairB_r, bone2Name=head, swing1LimitDegrees=40, swing2LimitDegrees=40, twistLimitDegrees=10}
]
